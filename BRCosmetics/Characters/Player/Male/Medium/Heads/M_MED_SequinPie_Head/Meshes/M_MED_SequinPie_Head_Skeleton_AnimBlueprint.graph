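# Animation Blueprint for the SequinPie player head cosmetic (medium male), making the head mesh follow the mesh it is attached to.
# It copies the attached parent's pose and curves into a cached pose.
# A per-bone layered blend applies an unconnected pose input at 0.7 weight from neck_01 down, excluding the head branch; its second layer has no bone filters.
# The event graph holds only placeholder nodes, with no logic.
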AnimBlueprint /BRCosmetics/Characters/Player/Male/Medium/Heads/M_MED_SequinPie_Head/Meshes/M_MED_SequinPie_Head_Skeleton_AnimBlueprint
target_skeleton=&/M_MED_SequinPie_Head_Skeleton
graphs [
 AnimGraph {nodes [
   n1 = Layered blend per bone!(
    BasePose=Use cached pose 'headJointCache'()
    BlendPoses_0
    BlendPoses_1:Pose Link Structure=Use cached pose 'headJointCache'()
    BlendWeights_0=0.7
    BlendWeights_1:float=1
   ) -> (Pose) {
    blendPoses [{linkId=-1, sourceLinkId=-1}, {linkId=-1, sourceLinkId=-1}]
    blendWeights [1,1]
    layerSetup [{branchFilters [{blendDepth=1, boneName=neck_01}, {blendDepth=-1, boneName=head}]}, {branchFilters []}]
   }
   Output Pose(Result=@n1)
   Save cached pose 'headJointCache'(Pose=Copy Pose From Mesh() {bCopyCurves=true, bUseAttachedParent=true}) {cacheName=headJointCache}
  ]}
 EventGraph {nodes [
   TryGetPawnOwner()
   Event BlueprintUpdateAnimation()
  ]}
]
variables []
